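; Burst-style gameplay cue notify for the HighTower WarmWest item's hand launch, playing a one-shot visual and audio effect.
; OnExecute passes the cue's location and normal to an Activated event as position and direction.
; Activated spawns a template Niagara system rotated along that direction and a template sound at the same location, sharing one auto-destroy setting.
; The Niagara component is stored, while the trigger audio variable is set empty rather than to the spawned sound.
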(event Custom|Activated
  (params (Location :type Vector) (Direction :type Vector))
  ;[1 1 1] When GCNL is activated
  (bind _returnvalue (Niagara|SpawnSystematLocation self (Variables|HighTowerWarmWestDefaults|GetTriggerParticleSystemTemplate) Location (Math|Conversions|RotationFromXVector Direction) "1.000000,1.000000,1.000000" (Variables|HighTowerWarmWestDefaults|GetAutoDestroyComponents)))
  (Variables|HighTowerWarmWestDefaultComponents|SetTriggerParticleSystem _returnvalue)
  (Audio|SpawnSoundatLocation :WorldContextObject self :Sound (Variables|HighTowerWarmWestDefaults|GetTriggerAudioTemplate) :Location Location :VolumeMultiplier 1.0 :PitchMultiplier 1.0 :StartTime 0.0 :bAutoDestroy (Variables|HighTowerWarmWestDefaults|GetAutoDestroyComponents))
  (Variables|HighTowerWarmWestDefaultComponents|SetTriggerAudio 0))

(fn OnExecute
  (params (MyTarget :type Actor%20Object%20Reference) (Parameters :type Gameplay%20Cue%20Parameters%20Structure%20%28by%20ref%29))
  (bind (_normalizedmagnitude _rawmagnitude _effectcontext _matchedtagname _originaltag _aggregatedsourcetags _aggregatedtargettags _location _normal) (Ability|GameplayCue|BreakGameplayCueParameters Parameters))
  (CallFunction|Activated _location _normal)
  (return true))

(fn ConstructionScript)
